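; Building prop Blueprint for a DeepInk store shelf whose T-shirt set mesh can be reskinned through building texture data.
; The Construction Script calls the parent, then reads the texture data array and checks that a second entry (index 1) exists.
; If that entry has a valid override material, it applies it to slot 0 of the SM_Commerce_TShirtSet_B component.
; BeginPlay, overlap and Tick events are empty.
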(event EventBeginPlay)

(event Collision|EventActorBeginOverlap
  (params (OtherActor :type Actor%20Object%20Reference)))

(event EventTick
  (params (DeltaSeconds :type Float%20%28single-precision%29)))

(fn ConstructionScript
  (|CallParentFunction)
  (bind _outtexturedataarray (Building|GetBuildingTextureData))
  ;[1 1 1] We want this initial check to match the index we are trying to sample.
  ;
  ; So if you want to sample the first slot, use number zero
  ; Second slot, use number one.
  ;
  (if (> (Utilities|Array|Length _outtexturedataarray) 1)
    ;[1 1 1] Checking that the chosen TD override actually has a material assigned before attempting to apply it.
    (if (Utilities|IsValid (Variables|/Script/FortniteGame.BuildingTextureData|GetOverrideMaterial (Utilities|Array|Get(acopy) _outtexturedataarray 1)) :!class /Script/Engine.KismetSystemLibrary)
      (Rendering|Material|SetMaterial (Variables|Default|GetSM_Commerce_TShirtSet_B) 0 (Variables|/Script/FortniteGame.BuildingTextureData|GetOverrideMaterial (Utilities|Array|Get(acopy) _outtexturedataarray 1))))))
