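; BP_DeepInk_StoreShelf_E is a Fortnite store shelf building prop that fills itself through socket fillers and supports a material override.
; Its construction script calls the parent, reconstructs all socket fillers, then reads the building texture data.
; If the second texture data entry has a valid override material, that material is applied to slot 0 of four instanced mesh and socket filler components.
; BeginPlay, overlap and Tick events are empty.
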(event EventBeginPlay)

(event Collision|EventActorBeginOverlap
  (params (OtherActor :type Actor%20Object%20Reference)))

(event EventTick
  (params (DeltaSeconds :type Float%20%28single-precision%29)))

(fn ConstructionScript
  (|CallParentFunction)
  (Class|SocketFillerFunctionLibrary|ReconstructAllSocketFillers self)
  (bind _outtexturedataarray (Building|GetBuildingTextureData))
  (if (> (Utilities|Array|Length _outtexturedataarray) 1)
    (if (Utilities|IsValid (Variables|/Script/FortniteGame.BuildingTextureData|GetOverrideMaterial (Utilities|Array|Get(acopy) _outtexturedataarray 1)) :!class /Script/Engine.KismetSystemLibrary)
      (Rendering|Material|SetMaterial (Utilities|Array|MakeArray (Variables|Default|GetInstancedStaticMesh) (Variables|Default|GetISM_SocketFiller) (Variables|Default|GetInstancedStaticMesh1) (Variables|Default|GetISM_SocketFiller1)) 0 (Variables|/Script/FortniteGame.BuildingTextureData|GetOverrideMaterial (Utilities|Array|Get(acopy) _outtexturedataarray 1))))))
